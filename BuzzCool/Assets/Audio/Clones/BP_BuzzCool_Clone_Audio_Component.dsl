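; Audio component for Buzz Cool clone pawns that manages spawn voice-over, recurring voice lines and death cleanup.
; BeginPlay assigns the pawn's voice sound and plays an attached spawn line, whose completion hands off to randomly delayed timer playback for non-clustered clones.
; Finished lines request the next timer; pawn destruction broadcasts OnDeath and fades out playing audio.
; SetVOSoundComponent feeds wave arrays into the sound's "Waves" parameter.
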(event EventBeginPlay
  ;[1 1 1] Set Owner Actor
  (bind _aspawn_wild_estate_buzz_cool (Utilities|Casting|CastToPawn_WildEstate_BuzzCool (Components|GetOwner)))
  (bind _output_get (Variables|Default|SetBuzz%20Cool%20Clone%20Pawn _aspawn_wild_estate_buzz_cool))
  ;[1 1 1] Initiate audio setup
  (Audio|Components|Audio|SetSound (Variables|/BuzzCool/NPCs/BuzzCool_Clones/Pawn/Pawn_WildEstate_BuzzCool.Pawn_WildEstate_BuzzCool_C|GetCloneVoiceAudio _output_get) (Variables|Default|GetVoiceSound))
  (Game|AssignOnDestroyed (Variables|Default|GetBuzz%20Cool%20Clone%20Pawn) (EventDispatchers|CreateEvent :!event Custom|OnActorDestroyed))
  (bind _returnvalue (Audio|SpawnSoundAttached :Sound (Variables|Default|GetSpawnVO) :AttachToComponent (Variables|/BuzzCool/NPCs/BuzzCool_Clones/Pawn/Pawn_WildEstate_BuzzCool.Pawn_WildEstate_BuzzCool_C|GetCloneVoiceAudio (Variables|Default|GetBuzz%20Cool%20Clone%20Pawn)) :AttachPointName "None" :LocationType "KeepRelativeOffset" :bStopWhenAttachedToDestroyed false :VolumeMultiplier 1.0 :PitchMultiplier 1.0 :StartTime 0.0 :bAutoDestroy false))
  (Utilities|IsValid _returnvalue
    (:Is%20Valid
      (EventDispatchers|AssignOnAudioFinished _returnvalue (delegate Custom|OnAudioFinished_Event)))
    (:Is%20Not%20Valid
      (CallFunction|OnAudioCall))))

(event Custom|OnAudioFinished_Event
  ;[1 1 1] Initiate audio setup
  (Components|DestroyComponent (stale EventBeginPlay _returnvalue))
  (CallFunction|OnAudioCall))

(event Custom|OnActorDestroyed
  (params (DestroyedActor :type Actor%20Object%20Reference))
  (bind _buzz_cool_clone_pawn (Variables|Default|GetBuzz%20Cool%20Clone%20Pawn))
  (bind _clonevoiceaudio (Variables|/BuzzCool/NPCs/BuzzCool_Clones/Pawn/Pawn_WildEstate_BuzzCool.Pawn_WildEstate_BuzzCool_C|GetCloneVoiceAudio _buzz_cool_clone_pawn))
  (CallFunction|UpdateAudio false)
  (Default|CallOnDeath DestroyedActor)
  (Variables|Default|SetIsDead true)
  (if (Audio|Components|Audio|IsPlaying _clonevoiceaudio)
    (Audio|Components|Audio|FadeOut _clonevoiceaudio 0.15 0.0 "Linear" :!class /Script/Engine.AudioComponent)))

(event Custom|OnAudioCall
  ;[1 1 1] Play Audio - For none Cluster Pawns
  ;;[1 1 1] Attempt to play audio again if allowed, once the current line stops.
  (CallFunction|UpdateAudio true))

(event Custom|UpdateAudio
  (params (Play :type Boolean))
  (bind _handleplayaudio (Variables|Default|GetHandlePlayAudio))
  ;[1 1 1] Play Audio - For none Cluster Pawns
  ;;[1 1 1] Play single audio line at random time once, but stop if clone enters audio cluster zone.
  (if Play
    (if (not (Variables|Default|GetIsAudioClustered))
      (bind _returnvalue (Utilities|Time|SetTimerbyEvent (EventDispatchers|CreateEvent :!event Custom|PlayAudio) (Math|Random|RandomFloatinRange 2.0 5.0)))
      (Variables|Default|SetHandlePlayAudio _returnvalue)))
  (if (Utilities|Time|IsValidTimerHandle _handleplayaudio)
    (Utilities|Time|ClearandInvalidateTimerbyHandle _handleplayaudio)))

(event Custom|PlayAudio
  ;[1 1 1] Play Audio - For none Cluster Pawns
  ;;[1 1 1] Play audio and watch for when each audio triggers finish.
  (bind _clonevoiceaudio (Variables|/BuzzCool/NPCs/BuzzCool_Clones/Pawn/Pawn_WildEstate_BuzzCool.Pawn_WildEstate_BuzzCool_C|GetCloneVoiceAudio (Variables|Default|GetBuzz%20Cool%20Clone%20Pawn))
    (:then
      (Audio|Components|Audio|Play _clonevoiceaudio 0.0)
      (EventDispatchers|BindEventtoOnAudioFinished _clonevoiceaudio (EventDispatchers|CreateEvent :!event Custom|OnAudioCall)))
    (:else)))

(fn SetVOSoundComponent
  (params (VO%20Sound%20Waves :type Array%20of%20Object%20References))
  (Audio|Parameter|SetObjectArrayParameter (Variables|/BuzzCool/NPCs/BuzzCool_Clones/Pawn/Pawn_WildEstate_BuzzCool.Pawn_WildEstate_BuzzCool_C|GetCloneVoiceAudio (Variables|Default|GetBuzz%20Cool%20Clone%20Pawn)) "Waves" VO%20Sound%20Waves))
