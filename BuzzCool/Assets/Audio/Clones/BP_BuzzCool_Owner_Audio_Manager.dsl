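; BP_BuzzCool_Owner_Audio_Manager plays one 2D VO sound on the locally viewed owner pawn while BuzzCool clone audio clusters are active.
; BeginPlay registers the 2D cluster configuration, then listens on the game state event router for cluster payloads, linking each cluster to the manager and tracking it.
; The first cluster starts the sound attached to the pawn mesh; losing the last, or EndPlay, fires a "Stop" trigger so audio can finish.
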(event EventBeginPlay
  ;[1 1 1] Set Owner Actor (Local Only)
  (bind _asfort_pawn (Utilities|Casting|CastToFortPawn (Components|GetOwner)))
  (if (Pawn|IsLocallyViewed _asfort_pawn)
    (Variables|Default|SetOwnerPawn _asfort_pawn)
    ;[1 1 1] Register Cluster
    (bind _returnvalue (AudioClusters|RegisterClusterConfigurations "/BuzzCool/Assets/Audio/Clones/AudioCluster/ACC_BuzzCool_2D.ACC_BuzzCool_2D"))
    (Variables|Default|SetClusterConfig _returnvalue)
    (Class|FortAsyncActionGameStateReady|GameStateReadyAsync
      (:OnReady
        ;[1 1 1] Listen for payload update when AudioCluster is created/updated (2D Cluster)
        (bind _returnvalue_2 (GameplayEventRouter|GetEventRouter "/Script/FortniteGame.FortGameState" (Components|GetOwner)))
        (bind (_asynctaskproxy _payload) (GameplayEventRouter|StartListeningToEvent _returnvalue_2 "/BuzzCool/Assets/Audio/Clones/AudioCluster/Struct_BuzzCool_AudioCluster.Struct_BuzzCool_AudioCluster")
          (:then
            (Variables|Default|SetAsyncTaskRef _asynctaskproxy))
          (:OnEventReceived
            (Class|ACBPlaySoundBuzzCoolClones2D|UpdateOwnerAudioManager (Utilities|Struct|BreakStructBuzzCoolAudioCluster _payload) self)
            (Utilities|Array|AddUnique (Variables|Default|GetAudioClusterReferences) (Utilities|Struct|BreakStructBuzzCoolAudioCluster _payload))
            (CallFunction|PlayAudio)))))))

(event EventEndPlay
  (params (EndPlayReason :type EEndPlayReason%20Enum :default "Destroyed"))
  (bind _clusterconfig (Variables|Default|GetClusterConfig))
  ;[1 1 1] Stop systems when endplay happens.
  (CallFunction|UpdateAudioState false)
  (if (AudioClusters|IsAudioClusterConfigurationsHandleValid _clusterconfig)
    (AudioClusters|UnregisterClusterConfigurations _clusterconfig)
    (bind _asynctaskref (Variables|Default|GetAsyncTaskRef)
      (:then
        (GameplayEventRouter|StopListeningtoEvent _asynctaskref))
      (:else))))

(event Custom|PlayAudio
  ;[1 1 1] Play Sound if not playing, during Audio Cluster creation - Just takes 1 Audio Cluster to initialize audio to play.
  (if (Utilities|Array|IsNotEmpty (Variables|Default|GetAudioClusterReferences))
    (if (not (Variables|Default|GetIsPlaying))
      (CallFunction|UpdateAudioState true))))

(event Custom|AudioClusterDestroyed
  (params (Audio%20Cluster%20Behavior%20Reference :type ACB%20Play%20Sound%20Buzz%20Cool%20Clones%202D%20Object%20Reference))
  ;[1 1 1] Update from AudioCluster OnStop to remove from array. Then stop playing sound when no clusters are active.
  (Utilities|Array|RemoveItem (Variables|Default|GetAudioClusterReferences) Audio%20Cluster%20Behavior%20Reference)
  (if (Utilities|Array|IsEmpty (Variables|Default|GetAudioClusterReferences))
    (CallFunction|UpdateAudioState false)))

(fn UpdateAudioState
  (params (Play :type Boolean))
  (if Play
    ;[1 1 1] Play Sound
    (Variables|Default|SetIsPlaying true)
    (bind _2d_audio (Variables|Default|Get2D%20Audio)
      (:then
        (Audio|Parameter|ExecuteTriggerParameter _2d_audio "Stop"))
      (:else
        _))
    (bind _ownerpawn (Variables|Default|GetOwnerPawn)
      (:then
        (bind _returnvalue (Audio|SpawnSoundAttached (Variables|Default|GetVO%20Sound) (Variables|/Script/Engine.Character|GetMesh _ownerpawn)))
        (Variables|Default|Set2D%20Audio _returnvalue))
      (:else))
    (else
      ;[1 1 1] Stop Sound, but let it finish audio
      (Variables|Default|SetIsPlaying false)
      (bind _2d_audio_2 (Variables|Default|Get2D%20Audio)
        (:then
          (Audio|Parameter|ExecuteTriggerParameter _2d_audio_2 "Stop"))
        (:else)))))
