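# MetaSound patch that plays randomly chosen quad waves from an array, replaying with random delays while a ShouldPlay check passes.
# Trigger starts a random delay (only StartDelay.Min is wired, not Max), then a random array pick feeds the quad Wave Player.
# On finish, a random replay delay leads to the ShouldPlay compare, looping back through a deferred trigger variable or firing Finish.
# Play and finish events set IsPlayIng through a trigger route.
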MetaSoundPatch /BuzzCool/Assets/Audio/Clones/MSP_BuzzCool_Group_Quad_Player
inputs [
 ReplayDelay.Max {type=Time}
 ReplayDelay.Min {type=Time}
 ShouldPlay {type=Bool}
 StartDelay.Max {type=Time}
 StartDelay.Min {type=Time}
 Trigger {type=Trigger}
 Waves {type=WaveAsset:Array}
]
outputs [
 Finish {type=Trigger}
 IsPlayIng {type=Bool}
 Out Front Left {type=Audio}
 Out Front Right {type=Audio}
 Out Side Left {type=Audio}
 Out Side Right {type=Audio}
]
pages [Default {
  input_defaults {ReplayDelay.Max=0, ReplayDelay.Min=0, ShouldPlay=null, StartDelay.Max=0, StartDelay.Min=0, Trigger=null, Waves=null}
  nodes [
   n1 = UE.RandomTime(Min=^StartDelay.Min, Next=^Trigger)
   TriggerDelay_0 = UE.Trigger Delay(In=@n1.On Next, Delay Time=@n1.Value)
   n2 = VariableDeferredAccessor.Trigger(Variable=InitVariable.Trigger().Variable)
   TriggerAnyTriggerAny2_0 = "TriggerAny.Trigger Any (2)"(In 0=@TriggerDelay_0.Out, In 1=@n2.Value)
   ArrayRandomGet_0 = Array.Random Get:WaveAsset:Array(In Array=^Waves, Next=@TriggerAnyTriggerAny2_0.Out)
   WavePlayer_0 = UE.Wave Player:Quad(Play=@ArrayRandomGet_0.On Next, Wave Asset=@ArrayRandomGet_0.Value) {also {version [1,1]}}
   TriggerRouteTriggerRouteBool2_0 = "TriggerRoute.Trigger Route (Bool, 2):Bool"(Set 1=@WavePlayer_0.On Finished, Set 0=@WavePlayer_0.On Play) {also {defaults {Value 0="<MetaSoundBuilderResult.SUCCEEDED: 0>"}}}
   n3 = UE.RandomTime(Max=^ReplayDelay.Max, Min=^ReplayDelay.Min, Next=@WavePlayer_0.On Finished)
   TriggerDelay_1 = UE.Trigger Delay(In=@n3.On Next, Delay Time=@n3.Value)
   TriggerCompareClamp_0 = TriggerCompare.Clamp:Bool(A=^ShouldPlay, Compare=@TriggerDelay_1.Out) {also {defaults {B="<MetaSoundBuilderResult.SUCCEEDED: 0>"}}}
   VariableMutator.Trigger(Value=@TriggerCompareClamp_0.True, Variable=@n2.Variable)
  ]
  links [
   {dst="@Finish", src=TriggerCompareClamp_0.False}
   {dst="@IsPlayIng", src=TriggerRouteTriggerRouteBool2_0.Value}
   {dst="@Out Front Left", src=WavePlayer_0.Out Front Left}
   {dst="@Out Front Right", src=WavePlayer_0.Out Front Right}
   {dst="@Out Side Left", src=WavePlayer_0.Out Side Left}
   {dst="@Out Side Right", src=WavePlayer_0.Out Side Right}
  ]
 }]
