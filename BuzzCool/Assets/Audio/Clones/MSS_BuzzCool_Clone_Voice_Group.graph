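# One-shot stereo MetaSound source for BuzzCool clone group voices, drawing on throw, spawn, donut and zombie voice clips.
# On play, two group quad player patches share the wave array with differing start and replay delay ranges, and their front outputs are mixed to stereo.
# OnPlay and Stop set a shared ShouldPlay flag.
# When either player finishes, a count of still-playing players is compared to zero to fire OnFinished.
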MetaSoundSource /BuzzCool/Assets/Audio/Clones/MSS_BuzzCool_Clone_Voice_Group
inputs [
 Stop {type=Trigger}
 UE.Source.OnPlay {type=Trigger}
 Waves {type=WaveAsset:Array}
]
interfaces [UE.OutputFormat.Stereo:1.0, UE.Source.OneShot:1.0, UE.Source:1.1]
outputs [
 "UE.OutputFormat.Stereo.Audio:0" {type=Audio}
 "UE.OutputFormat.Stereo.Audio:1" {type=Audio}
 UE.Source.OneShot.OnFinished {type=Trigger}
]
pages [Default {
  input_defaults {
   Stop=null
   UE.Source.OnPlay=null
   Waves [
    /BuzzCool/Assets/Audio/Clones/Waves/vo_IP_WildEstate_BuzzCool_OnThrow_010
    /BuzzCool/Assets/Audio/Clones/Waves/vo_IP_WildEstate_BuzzCool_OnThrow_020
    /BuzzCool/Assets/Audio/Clones/Waves/vo_IP_WildEstate_BuzzCool_OnThrow_030
    /BuzzCool/Assets/Audio/Clones/Waves/vo_IP_WildEstate_BuzzCool_OnThrow_040
    /BuzzCool/Assets/Audio/Clones/Waves/vo_IP_WildEstate_BuzzCool_OnThrow_041
    /BuzzCool/Assets/Audio/Clones/Waves/vo_IP_WildEstate_BuzzCool_OnThrow_042
    /BuzzCool/Assets/Audio/Clones/Waves/vo_IP_WildEstate_BuzzCool_OnThrow_060
    /BuzzCool/Assets/Audio/Clones/Waves/vo_IP_WildEstate_BuzzCool_Spawn_020
    /BuzzCool/Assets/Audio/Clones/Waves/vo_IP_WildEstate_BuzzCool_Spawn_021
    /BuzzCool/Assets/Audio/Clones/Waves/vo_IP_WildEstate_BuzzCool_Spawn_022
    /BuzzCool/Assets/Audio/Clones/Waves/vo_IP_WildEstate_BuzzCool_Spawn_030
    /BuzzCool/Assets/Audio/Clones/Waves/vo_IP_WildEstate_BuzzCool_Spawn_031
    /BuzzCool/Assets/Audio/Clones/Waves/vo_IP_WildEstate_BuzzCool_Spawn_032
    /BuzzCool/Assets/Audio/Clones/Waves/vo_IP_WildEstate_BuzzCool_Spawn_050
    /BuzzCool/Assets/Audio/Clones/Waves/vo_IP_WildEstate_BuzzCool_MmmDonuts_010
    /BuzzCool/Assets/Audio/Clones/Waves/vo_IP_WildEstate_BuzzCool_MmmDonuts_012
    /BuzzCool/Assets/Audio/Clones/Waves/vo_IP_WildEstate_BuzzCool_NV_Zombie_010
    /BuzzCool/Assets/Audio/Clones/Waves/vo_IP_WildEstate_BuzzCool_NV_Zombie_011
    /BuzzCool/Assets/Audio/Clones/Waves/vo_IP_WildEstate_BuzzCool_NV_Zombie_012
    /BuzzCool/Assets/Audio/Clones/Waves/vo_IP_WildEstate_BuzzCool_NV_Zombie_013
   ]
  }
  nodes [
   TriggerRouteTriggerRouteBool2_0 = "TriggerRoute.Trigger Route (Bool, 2):Bool"(Set 1=^Stop, Set 0=^UE.Source.OnPlay) {also {defaults {Value 0="<MetaSoundBuilderResult.SUCCEEDED: 0>"}}}
   n1 = VariableAccessor.Bool(Variable=VariableMutator.Bool(Variable=InitVariable.Bool().Variable, Value=@TriggerRouteTriggerRouteBool2_0.Value).Variable)
   n2 = &/MSP_BuzzCool_Group_Quad_Player(Trigger=^UE.Source.OnPlay, Waves=^Waves, ShouldPlay=@n1.Value) {also {defaults {ReplayDelay.Max=4, ReplayDelay.Min=0.25, StartDelay.Max=1}}}
   n3 = &MSP_BuzzCool_Group_Quad_Player(Trigger=^UE.Source.OnPlay, Waves=^Waves, ShouldPlay=@n1.Value) {also {defaults {ReplayDelay.Max=3, ReplayDelay.Min=1, StartDelay.Max=1.5, StartDelay.Min=0.5}}}
   AudioMixerAudioMixerStereo2_0 = "AudioMixer.Audio Mixer (Stereo, 2)"(In 0 L=@n2.Out Front Left, In 0 R=@n2.Out Front Right, In 1 L=@n3.Out Front Left, In 1 R=@n3.Out Front Right)
   TriggerAnyTriggerAny2_0 = "TriggerAny.Trigger Any (2)"(In 0=@n2.Finish, In 1=@n3.Finish)
   n4 = UE.Add:Int32(PrimaryOperand=Convert.Int32:Bool(Bool=@n2.IsPlayIng).Int32, AdditionalOperands=Convert.Int32:Bool(Bool=@n3.IsPlayIng).Int32)
   n5 = InitVariable.Int32()
   TriggerCompareClamp_0 = TriggerCompare.Clamp:Int32(
    Compare=@TriggerAnyTriggerAny2_0.Out
    A=VariableAccessor.Int32(Variable=VariableMutator.Int32(Value=@n4.Out, Variable=@n5.Variable).Variable).Value
   ) {also {defaults {B=0}}}
  ]
  links [
   {dst="@UE.OutputFormat.Stereo.Audio:0", src=AudioMixerAudioMixerStereo2_0.Out L}
   {dst="@UE.OutputFormat.Stereo.Audio:1", src=AudioMixerAudioMixerStereo2_0.Out R}
   {dst="@UE.Source.OneShot.OnFinished", src=TriggerCompareClamp_0.True}
  ]
 }]
properties {
 assetImportData=&self:AssetImportData
 attenuationSettings=/BuzzCool/Assets/Audio/Clones/Settings/ATT_BuzzCool_Voice_Group
 bOverrideConcurrency=true
 maxDistance=2800
 numChannels=2
 outputFormat=Stereo
 sampleRate=48000
 soundClassObject=&Fort_VO_NPC
 virtualizationMode=PlayWhenSilent
 volume=0.8
 concurrencyOverrides {maxCount=3, voiceStealReleaseTime=0.5, volumeScaleAttackTime=1}
}
