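# MetaSound patch that pans a four-channel (front and side left/right) signal around a quad layout by Azimuth.
# Azimuth drives four crossfades peaking a quarter turn apart, with Width setting each gain's off-peak level.
# Four mono mixers apply these gains in permuted order to all inputs, rotating the field, then scale by Gain into each output.
# Additional per-input multiply chains using averaged gain pairs connect to no output.
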MetaSoundPatch /FortMetasound/Patches/MSP_AzimuthPositionalPannerQuad
inputs [
 Azimuth {type=Float}
 Gain {type=Float}
 In Front Left {type=Audio}
 In Front Right {type=Audio}
 In Side Left {type=Audio}
 In Side Right {type=Audio}
 Width {type=Float}
]
outputs [
 Out Front Left {type=Audio}
 Out Front Right {type=Audio}
 Out Side Left {type=Audio}
 Out Side Right {type=Audio}
]
pages [Default {
  input_defaults {Azimuth=0, Gain=1, In Front Left=null, In Front Right=null, In Side Left=null, In Side Right=null, Width=0}
  nodes [
   CrossfadeTriggerRouteFloat3_3 = "Crossfade.Trigger Route (Float, 3):Float"(In 0=^Width, In 2=^Width, Crossfade Value=MapRange.MapRange:Float(In=^Azimuth) {also {defaults {In Range B=-180, Out Range B=2}}}.Out Value) {also {defaults {In 1=1}}}
   n1 = VariableAccessor.Float(Variable=VariableMutator.Float(Value=@CrossfadeTriggerRouteFloat3_3.Out, Variable=InitVariable.Float().Variable).Variable)
   n2 = VariableAccessor.Float(Variable=@n1.Variable)
   n3 = VariableAccessor.Float(Variable=@n2.Variable)
   CrossfadeTriggerRouteFloat3_2 = "Crossfade.Trigger Route (Float, 3):Float"(
    In 0=^Width
    In 2=^Width
    Crossfade Value=UE.Add:Float(
     AdditionalOperands=MapRange.MapRange:Float(In=^Azimuth) {also {defaults {In Range A=-90, In Range B=0}}}.Out Value
     PrimaryOperand=MapRange.MapRange:Float(In=^Azimuth) {also {defaults {In Range B=90}}}.Out Value
    ).Out
   ) {also {defaults {In 1=1}}}
   n4 = VariableAccessor.Float(Variable=VariableMutator.Float(Value=@CrossfadeTriggerRouteFloat3_2.Out, Variable=InitVariable.Float().Variable).Variable)
   n5 = VariableAccessor.Float(Variable=@n4.Variable)
   n6 = VariableAccessor.Float(Variable=@n5.Variable)
   CrossfadeTriggerRouteFloat3_0 = "Crossfade.Trigger Route (Float, 3):Float"(In 0=^Width, In 2=^Width, Crossfade Value=MapRange.MapRange:Float(In=^Azimuth) {also {defaults {In Range B=180, Out Range B=2}}}.Out Value) {also {defaults {In 1=1}}}
   n7 = VariableAccessor.Float(Variable=VariableMutator.Float(Value=@CrossfadeTriggerRouteFloat3_0.Out, Variable=InitVariable.Float().Variable).Variable)
   n8 = VariableAccessor.Float(Variable=@n7.Variable)
   n9 = VariableAccessor.Float(Variable=@n8.Variable)
   CrossfadeTriggerRouteFloat3_1 = "Crossfade.Trigger Route (Float, 3):Float"(
    In 0=^Width
    In 2=^Width
    Crossfade Value=UE.Add:Float(
     PrimaryOperand=MapRange.MapRange:Float(In=^Azimuth) {also {defaults {In Range A=90, In Range B=180}}}.Out Value
     AdditionalOperands=MapRange.MapRange:Float(In=^Azimuth) {also {defaults {In Range A=-180, In Range B=-90, Out Range A=1, Out Range B=0}}}.Out Value
    ).Out
   ) {also {defaults {In 1=1}}}
   n10 = VariableAccessor.Float(Variable=VariableMutator.Float(Value=@CrossfadeTriggerRouteFloat3_1.Out, Variable=InitVariable.Float().Variable).Variable)
   n11 = VariableAccessor.Float(Variable=@n10.Variable)
   n12 = VariableAccessor.Float(Variable=@n11.Variable)
   AudioMixerAudioMixerMono4_2 = "AudioMixer.Audio Mixer (Mono, 4)"(
    In 0=^In Front Left
    In 1=^In Front Right
    In 2=^In Side Left
    In 3=^In Side Right
    Gain 3=@n3.Value
    Gain 2=@n6.Value
    Gain 0=@n9.Value
    Gain 1=@n12.Value
   )
   Multiply_10 = UE.Multiply:Audio by Float(AdditionalOperands=^Gain, PrimaryOperand=@AudioMixerAudioMixerMono4_2.Out)
   n13 = UE.Add:Float(PrimaryOperand=@n2.Value, AdditionalOperands=@n10.Value)
   UE.Multiply:Audio by Float(
    AdditionalOperands=^Gain
    PrimaryOperand=UE.Multiply:Audio by Float(PrimaryOperand=^In Side Left, AdditionalOperands=MapRange.MapRange:Float(In=@n13.Out) {also {defaults {In Range B=2}}}.Out Value).Out
   )
   n14 = UE.Add:Float(AdditionalOperands=@n11.Value, PrimaryOperand=@n8.Value)
   UE.Multiply:Audio by Float(
    AdditionalOperands=^Gain
    PrimaryOperand=UE.Multiply:Audio by Float(PrimaryOperand=^In Side Right, AdditionalOperands=MapRange.MapRange:Float(In=@n14.Out) {also {defaults {In Range B=2}}}.Out Value).Out
   )
   AudioMixerAudioMixerMono4_0 = "AudioMixer.Audio Mixer (Mono, 4)"(
    In 0=^In Front Left
    In 1=^In Front Right
    In 2=^In Side Left
    In 3=^In Side Right
    Gain 1=@n3.Value
    Gain 3=@n6.Value
    Gain 2=@n9.Value
    Gain 0=@n12.Value
   )
   Multiply_3 = UE.Multiply:Audio by Float(AdditionalOperands=^Gain, PrimaryOperand=@AudioMixerAudioMixerMono4_0.Out)
   n15 = UE.Add:Float(PrimaryOperand=@n5.Value, AdditionalOperands=@n7.Value)
   UE.Multiply:Audio by Float(
    AdditionalOperands=^Gain
    PrimaryOperand=UE.Multiply:Audio by Float(PrimaryOperand=^In Front Right, AdditionalOperands=MapRange.MapRange:Float(In=@n15.Out) {also {defaults {In Range B=2}}}.Out Value).Out
   )
   AudioMixerAudioMixerMono4_3 = "AudioMixer.Audio Mixer (Mono, 4)"(
    In 0=^In Front Left
    In 1=^In Front Right
    In 2=^In Side Left
    In 3=^In Side Right
    Gain 0=@n3.Value
    Gain 1=@n6.Value
    Gain 3=@n9.Value
    Gain 2=@n12.Value
   )
   Multiply_6 = UE.Multiply:Audio by Float(AdditionalOperands=^Gain, PrimaryOperand=@AudioMixerAudioMixerMono4_3.Out)
   AudioMixerAudioMixerMono4_1 = "AudioMixer.Audio Mixer (Mono, 4)"(
    In 0=^In Front Left
    In 1=^In Front Right
    In 2=^In Side Left
    In 3=^In Side Right
    Gain 2=@n3.Value
    Gain 0=@n6.Value
    Gain 1=@n9.Value
    Gain 3=@n12.Value
   )
   Multiply_7 = UE.Multiply:Audio by Float(AdditionalOperands=^Gain, PrimaryOperand=@AudioMixerAudioMixerMono4_1.Out)
   n16 = UE.Add:Float(AdditionalOperands=@n1.Value, PrimaryOperand=@n4.Value)
   UE.Multiply:Audio by Float(
    AdditionalOperands=^Gain
    PrimaryOperand=UE.Multiply:Audio by Float(PrimaryOperand=^In Front Left, AdditionalOperands=MapRange.MapRange:Float(In=@n16.Out) {also {defaults {In Range B=2}}}.Out Value).Out
   )
  ]
  links [
   {dst="@Out Side Left", src=Multiply_10.Out}
   {dst="@Out Side Right", src=Multiply_3.Out}
   {dst="@Out Front Right", src=Multiply_6.Out}
   {dst="@Out Front Left", src=Multiply_7.Out}
  ]
 }]
