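# Stereo MetaSound source for BuzzCool clone group NPC voices, drawing on a shared array of voice lines.
# Four quad player patches with differing start and replay delays run under a play/Stop flag, mixed per quadrant.
# Smoothed NumActors inputs scale each quadrant before a listener-azimuth quad panner folds to stereo.
# Player finish triggers check the summed playing count to fire OnFinished.
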MetaSoundSource /BuzzCool/Assets/Audio/Clones/MSS_BuzzCool_Clone_Voice_Group_2D
inputs [
 NumActors0 {type=Float}
 NumActors1 {type=Float}
 NumActors2 {type=Float}
 NumActors3 {type=Float}
 Stop {type=Trigger}
 UE.Attenuation.Distance {type=Float}
 UE.Listener.Orientation.Azimuth {type=Float}
 UE.Listener.Orientation.Elevation {type=Float}
 UE.Source.OnPlay {type=Trigger}
 Waves {type=WaveAsset:Array}
]
interfaces [UE.Attenuation:1.0, UE.Listener.Orientation:1.0, UE.OutputFormat.Stereo:1.0, UE.Source.OneShot:1.0, UE.Source:1.1]
outputs [
 "UE.OutputFormat.Stereo.Audio:0" {type=Audio}
 "UE.OutputFormat.Stereo.Audio:1" {type=Audio}
 UE.Source.OneShot.OnFinished {type=Trigger}
]
pages [Default {
  input_defaults {
   NumActors0=0
   NumActors1=0
   NumActors2=0
   NumActors3=0
   Stop=null
   UE.Attenuation.Distance=0
   UE.Listener.Orientation.Azimuth=0
   UE.Listener.Orientation.Elevation=0
   UE.Source.OnPlay=null
   Waves [
    /BuzzCool/Assets/Audio/Clones/Waves/vo_IP_WildEstate_BuzzCool_OnThrow_010
    /BuzzCool/Assets/Audio/Clones/Waves/vo_IP_WildEstate_BuzzCool_OnThrow_020
    /BuzzCool/Assets/Audio/Clones/Waves/vo_IP_WildEstate_BuzzCool_OnThrow_030
    /BuzzCool/Assets/Audio/Clones/Waves/vo_IP_WildEstate_BuzzCool_OnThrow_040
    /BuzzCool/Assets/Audio/Clones/Waves/vo_IP_WildEstate_BuzzCool_OnThrow_041
    /BuzzCool/Assets/Audio/Clones/Waves/vo_IP_WildEstate_BuzzCool_OnThrow_042
    /BuzzCool/Assets/Audio/Clones/Waves/vo_IP_WildEstate_BuzzCool_OnThrow_060
    /BuzzCool/Assets/Audio/Clones/Waves/vo_IP_WildEstate_BuzzCool_Spawn_020
    /BuzzCool/Assets/Audio/Clones/Waves/vo_IP_WildEstate_BuzzCool_Spawn_021
    /BuzzCool/Assets/Audio/Clones/Waves/vo_IP_WildEstate_BuzzCool_Spawn_022
    /BuzzCool/Assets/Audio/Clones/Waves/vo_IP_WildEstate_BuzzCool_Spawn_030
    /BuzzCool/Assets/Audio/Clones/Waves/vo_IP_WildEstate_BuzzCool_Spawn_031
    /BuzzCool/Assets/Audio/Clones/Waves/vo_IP_WildEstate_BuzzCool_Spawn_032
    /BuzzCool/Assets/Audio/Clones/Waves/vo_IP_WildEstate_BuzzCool_Spawn_050
    /BuzzCool/Assets/Audio/Clones/Waves/vo_IP_WildEstate_BuzzCool_MmmDonuts_010
    /BuzzCool/Assets/Audio/Clones/Waves/vo_IP_WildEstate_BuzzCool_MmmDonuts_012
    /BuzzCool/Assets/Audio/Clones/Waves/vo_IP_WildEstate_BuzzCool_NV_Zombie_010
    /BuzzCool/Assets/Audio/Clones/Waves/vo_IP_WildEstate_BuzzCool_NV_Zombie_011
    /BuzzCool/Assets/Audio/Clones/Waves/vo_IP_WildEstate_BuzzCool_NV_Zombie_012
    /BuzzCool/Assets/Audio/Clones/Waves/vo_IP_WildEstate_BuzzCool_NV_Zombie_013
   ]
  }
  nodes [
   TriggerRouteTriggerRouteBool2_0 = "TriggerRoute.Trigger Route (Bool, 2):Bool"(Set 1=^Stop, Set 0=^UE.Source.OnPlay) {also {defaults {Value 0="<MetaSoundBuilderResult.SUCCEEDED: 0>"}}}
   n1 = VariableAccessor.Bool(Variable=VariableMutator.Bool(Variable=InitVariable.Bool().Variable, Value=@TriggerRouteTriggerRouteBool2_0.Value).Variable)
   n2 = &/MSP_BuzzCool_Group_Quad_Player(Trigger=^UE.Source.OnPlay, Waves=^Waves, ShouldPlay=@n1.Value) {also {defaults {ReplayDelay.Max=4, ReplayDelay.Min=0.5, StartDelay.Max=3, StartDelay.Min=0.5}}}
   n3 = &MSP_BuzzCool_Group_Quad_Player(Trigger=^UE.Source.OnPlay, Waves=^Waves, ShouldPlay=@n1.Value) {also {defaults {ReplayDelay.Max=4, ReplayDelay.Min=0.5, StartDelay.Max=1}}}
   n4 = &MSP_BuzzCool_Group_Quad_Player(Trigger=^UE.Source.OnPlay, Waves=^Waves, ShouldPlay=@n1.Value) {also {defaults {ReplayDelay.Max=5, ReplayDelay.Min=1, StartDelay.Max=2, StartDelay.Min=1}}}
   n5 = &MSP_BuzzCool_Group_Quad_Player(Trigger=^UE.Source.OnPlay, Waves=^Waves, ShouldPlay=@n1.Value) {also {defaults {ReplayDelay.Max=4, ReplayDelay.Min=0.5, StartDelay.Max=5, StartDelay.Min=1.5}}}
   AudioMixerAudioMixerMono4_1 = "AudioMixer.Audio Mixer (Mono, 4)"(In 2=@n2.Out Side Right, In 0=@n3.Out Side Right, In 1=@n4.Out Side Right, In 3=@n5.Out Side Right) {also {defaults {Gain 0=0.5, Gain 1=0.5, Gain 2=0.5, Gain 3=0.5}}}
   n6 = UE.InterpTo:Audio(Target=^NumActors1)
   AudioMixerAudioMixerMono4_0 = "AudioMixer.Audio Mixer (Mono, 4)"(In 2=@n2.Out Front Left, In 0=@n3.Out Front Left, In 1=@n4.Out Front Left, In 3=@n5.Out Front Left) {also {defaults {Gain 0=0.5, Gain 1=0.5, Gain 2=0.5, Gain 3=0.5}}}
   n7 = UE.InterpTo:Audio(Target=^NumActors0)
   AudioMixerAudioMixerMono4_2 = "AudioMixer.Audio Mixer (Mono, 4)"(In 2=@n2.Out Side Left, In 0=@n3.Out Side Left, In 1=@n4.Out Side Left, In 3=@n5.Out Side Left) {also {defaults {Gain 0=0.5, Gain 1=0.5, Gain 2=0.5, Gain 3=0.5}}}
   n8 = UE.InterpTo:Audio(Target=^NumActors2)
   AudioMixerAudioMixerMono4_3 = "AudioMixer.Audio Mixer (Mono, 4)"(In 2=@n2.Out Front Right, In 0=@n3.Out Front Right, In 1=@n4.Out Front Right, In 3=@n5.Out Front Right) {also {defaults {Gain 0=0.5, Gain 1=0.5, Gain 2=0.5, Gain 3=0.5}}}
   n9 = UE.InterpTo:Audio(Target=^NumActors3)
   n10 = /FortMetasound/Patches/MSP_AzimuthPositionalPannerQuad(
    Azimuth=^UE.Listener.Orientation.Azimuth
    In Side Right=UE.Multiply:Audio by Float(
     PrimaryOperand=@AudioMixerAudioMixerMono4_1.Out
     AdditionalOperands=MapRange.MapRange:Float(In=@n6.Value) {also {defaults {In Range B=10, Out Range B=5}}}.Out Value
    ).Out
    In Front Left=UE.Multiply:Audio by Float(
     PrimaryOperand=@AudioMixerAudioMixerMono4_0.Out
     AdditionalOperands=MapRange.MapRange:Float(In=@n7.Value) {also {defaults {In Range B=10, Out Range B=5}}}.Out Value
    ).Out
    In Side Left=UE.Multiply:Audio by Float(
     PrimaryOperand=@AudioMixerAudioMixerMono4_2.Out
     AdditionalOperands=MapRange.MapRange:Float(In=@n8.Value) {also {defaults {In Range B=10, Out Range B=5}}}.Out Value
    ).Out
    In Front Right=UE.Multiply:Audio by Float(
     PrimaryOperand=@AudioMixerAudioMixerMono4_3.Out
     AdditionalOperands=MapRange.MapRange:Float(In=@n9.Value) {also {defaults {In Range B=10, Out Range B=5}}}.Out Value
    ).Out
   ) {also {defaults {Gain=0.5}}}
   AudioMixerAudioMixerStereo2_0 = "AudioMixer.Audio Mixer (Stereo, 2)"(In 0 L=@n10.Out Front Left, In 0 R=@n10.Out Front Right, In 1 L=@n10.Out Side Left, In 1 R=@n10.Out Side Right)
   TriggerAnyTriggerAny4_0 = "TriggerAny.Trigger Any (4)"(In 2=@n2.Finish, In 0=@n3.Finish, In 1=@n4.Finish, In 3=@n5.Finish)
   n11 = UE.Add:Int32(
    PrimaryOperand=UE.Add:Int32(AdditionalOperands=Convert.Int32:Bool(Bool=@n4.IsPlayIng).Int32, PrimaryOperand=Convert.Int32:Bool(Bool=@n3.IsPlayIng).Int32).Out
    AdditionalOperands=UE.Add:Int32(AdditionalOperands=Convert.Int32:Bool(Bool=@n5.IsPlayIng).Int32, PrimaryOperand=Convert.Int32:Bool(Bool=@n2.IsPlayIng).Int32).Out
   )
   n12 = InitVariable.Int32()
   TriggerCompareClamp_0 = TriggerCompare.Clamp:Int32(
    Compare=@TriggerAnyTriggerAny4_0.Out
    A=VariableAccessor.Int32(Variable=VariableMutator.Int32(Value=@n11.Out, Variable=@n12.Variable).Variable).Value
   )
  ]
  links [
   {dst="@UE.OutputFormat.Stereo.Audio:0", src=AudioMixerAudioMixerStereo2_0.Out L}
   {dst="@UE.OutputFormat.Stereo.Audio:1", src=AudioMixerAudioMixerStereo2_0.Out R}
   {dst="@UE.Source.OneShot.OnFinished", src=TriggerCompareClamp_0.True}
  ]
 }]
properties {
 assetImportData=&self:AssetImportData
 bOverrideConcurrency=true
 numChannels=2
 outputFormat=Stereo
 sampleRate=48000
 soundClassObject=&Fort_VO_NPC
 virtualizationMode=PlayWhenSilent
 volume=0.8
 concurrencyOverrides {maxCount=2}
}
