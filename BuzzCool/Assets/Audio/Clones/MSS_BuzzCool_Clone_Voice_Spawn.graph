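# One-shot stereo MetaSound source that plays a clone spawn voice line, picked from jump and throw vocal waves.
# On play, a short trigger delay leads into a random array pick, then chained random rolls for gain and pitch shift before the Wave Player starts.
# The gain scales both channels, and playback end finishes the source.
# It uses BuzzCool voice attenuation, an NPC voice sound class and a concurrency cap.
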MetaSoundSource /BuzzCool/Assets/Audio/Clones/MSS_BuzzCool_Clone_Voice_Spawn
inputs [
 UE.Source.OnPlay {type=Trigger}
 Waves {type=WaveAsset:Array}
]
interfaces [UE.OutputFormat.Stereo:1.0, UE.Source.OneShot:1.0, UE.Source:1.1]
outputs [
 "UE.OutputFormat.Stereo.Audio:0" {type=Audio}
 "UE.OutputFormat.Stereo.Audio:1" {type=Audio}
 UE.Source.OneShot.OnFinished {type=Trigger}
]
pages [Default {
  input_defaults {
   UE.Source.OnPlay=null
   Waves [
    /BuzzCool/Assets/Audio/Clones/Waves/vo_IP_WildEstate_SilentNovel_NV_Jump_014
    /BuzzCool/Assets/Audio/Clones/Waves/vo_IP_WildEstate_SilentNovel_NV_Jump_013
    /BuzzCool/Assets/Audio/Clones/Waves/vo_IP_WildEstate_SilentNovel_NV_Jump_012
    /BuzzCool/Assets/Audio/Clones/Waves/vo_IP_WildEstate_SilentNovel_NV_Jump_011
    /BuzzCool/Assets/Audio/Clones/Waves/vo_IP_WildEstate_SilentNovel_NV_Jump_010
    /BuzzCool/Assets/Audio/Clones/Waves/vo_IP_WildEstate_SilentNovel_NV_Throw_010
    /BuzzCool/Assets/Audio/Clones/Waves/vo_IP_WildEstate_SilentNovel_NV_Throw_011
    /BuzzCool/Assets/Audio/Clones/Waves/vo_IP_WildEstate_SilentNovel_NV_Throw_012
    /BuzzCool/Assets/Audio/Clones/Waves/vo_IP_WildEstate_SilentNovel_NV_Throw_013
    /BuzzCool/Assets/Audio/Clones/Waves/vo_IP_WildEstate_SilentNovel_NV_Throw_014
   ]
  }
  nodes [
   TriggerDelay_0 = UE.Trigger Delay(In=^UE.Source.OnPlay) {also {defaults {Delay Time=0.25}}}
   ArrayRandomGet_0 = Array.Random Get:WaveAsset:Array(In Array=^Waves, Next=@TriggerDelay_0.Out)
   n1 = UE.RandomFloat(Next=@ArrayRandomGet_0.On Next) {also {defaults {Max=1.1, Min=0.6}}}
   n2 = UE.Reroute(Value=UE.Reroute(Value=@n1.Value).Value)
   n3 = UE.RandomFloat(Next=@n1.On Next) {also {defaults {Min=-1}}}
   WavePlayer_0 = UE.Wave Player:Stereo(Wave Asset=@ArrayRandomGet_0.Value, Play=@n3.On Next, Pitch Shift=@n3.Value) {also {version [1,1]}}
   Multiply_0 = UE.Multiply:Audio by Float(AdditionalOperands=@n2.Value, PrimaryOperand=@WavePlayer_0.Out Left)
   Multiply_1 = UE.Multiply:Audio by Float(AdditionalOperands=@n2.Value, PrimaryOperand=@WavePlayer_0.Out Right)
  ]
  links [
   {dst="@UE.OutputFormat.Stereo.Audio:0", src=Multiply_0.Out}
   {dst="@UE.OutputFormat.Stereo.Audio:1", src=Multiply_1.Out}
   {dst="@UE.Source.OneShot.OnFinished", src=WavePlayer_0.On Finished}
  ]
 }]
properties {
 assetImportData=&self:AssetImportData
 attenuationSettings=/BuzzCool/Assets/Audio/Clones/Settings/ATT_BuzzCool_Voice
 bOverrideConcurrency=true
 maxDistance=2800
 numChannels=2
 outputFormat=Stereo
 sampleRate=48000
 soundClassObject=&Fort_VO_NPC
 virtualizationMode=PlayWhenSilent
 volume=1.5
 concurrencyOverrides {maxCount=5}
}
