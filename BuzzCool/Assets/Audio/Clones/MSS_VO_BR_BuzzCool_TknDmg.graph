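# Mono one-shot MetaSound for BuzzCool, apparently damage-taken voice, drawing on scream, burp and landing clips.
# A Trigger Filter, not wired to the exposed Probability input, either finishes immediately or passes through random delay, array shuffle and random pitch and volume into a wave player.
# Concurrency allows three per owner, preventing new ones, and a pre-effect send feeds a boss VO envelope-follower bus.
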MetaSoundSource /BuzzCool/Assets/Audio/Clones/MSS_VO_BR_BuzzCool_TknDmg
inputs [
 "Delay Max (Seconds)" {type=Time}
 "Delay Min (Seconds)" {type=Time}
 "Probability (0-1)" {type=Float}
 "Random Pitch Max (Semitones)" {type=Float}
 "Random Pitch Min (Semitones)" {type=Float}
 "Random Volume Change Max (Decibels)" {type=Float}
 "Random Volume Change Min (Decibels)" {type=Float}
 UE.Source.OnPlay {type=Trigger}
 Wave Array {type=WaveAsset:Array}
]
interfaces [UE.OutputFormat.Mono:1.0, UE.Source.OneShot:1.0, UE.Source:1.1]
outputs [
 "UE.OutputFormat.Mono.Audio:0" {type=Audio}
 UE.Source.OneShot.OnFinished {type=Trigger}
]
pages [Default {
  input_defaults {
   "Delay Max (Seconds)"=0
   "Delay Min (Seconds)"=0
   "Probability (0-1)"=0.66
   "Random Pitch Max (Semitones)"=0
   "Random Pitch Min (Semitones)"=0
   "Random Volume Change Max (Decibels)"=0
   "Random Volume Change Min (Decibels)"=0
   UE.Source.OnPlay=null
   Wave Array [
    /BuzzCool/Assets/Audio/Clones/Waves/vo_IP_WildEstate_SilentNovel_NV_Land_010
    /BuzzCool/Assets/Audio/Clones/Waves/vo_IP_WildEstate_GiantHomer_NV_Burp_011
    /BuzzCool/Assets/Audio/Clones/Waves/vo_IP_WildEstate_BuzzCool_NV_Scream_011
    /BuzzCool/Assets/Audio/Clones/Waves/vo_IP_WildEstate_BuzzCool_NV_Scream_010
    /BuzzCool/Assets/Audio/Clones/Waves/vo_IP_WildEstate_SilentNovel_NV_Land_011
    /BuzzCool/Assets/Audio/Clones/Waves/vo_IP_WildEstate_SilentNovel_NV_Land_012
    /BuzzCool/Assets/Audio/Clones/Waves/vo_IP_WildEstate_SilentNovel_NV_Land_013
    /BuzzCool/Assets/Audio/Clones/Waves/vo_IP_WildEstate_SilentNovel_NV_Land_014
    /BuzzCool/Assets/Audio/Clones/Waves/vo_IP_WildEstate_GiantHomer_NV_Burp_010
    /BuzzCool/Assets/Audio/Clones/Waves/vo_IP_WildEstate_GiantHomer_NV_Burp_012
    /BuzzCool/Assets/Audio/Clones/Waves/vo_IP_WildEstate_BuzzCool_NV_Scream_013
    /BuzzCool/Assets/Audio/Clones/Waves/vo_IP_WildEstate_BuzzCool_NV_Scream_014
   ]
  }
  nodes [
   TriggerFilter_0 = UE.Trigger Filter(Trigger=^UE.Source.OnPlay)
   n1 = UE.RandomTime(Max=^"Delay Max (Seconds)", Min=^"Delay Min (Seconds)", Next=@TriggerFilter_0.Heads)
   TriggerDelay_0 = UE.Trigger Delay(In=@n1.On Next, Delay Time=@n1.Value)
   n2 = Array.Shuffle:WaveAsset:Array(In Array=^Wave Array, Next=@TriggerDelay_0.Out) {also {defaults {Enable Shared State="<MetaSoundBuilderResult.SUCCEEDED: 0>"}}}
   n3 = UE.RandomFloat(Max=^"Random Pitch Max (Semitones)", Min=^"Random Pitch Min (Semitones)", Next=@n2.On Next)
   n4 = UE.RandomFloat(Max=^"Random Volume Change Max (Decibels)", Min=^"Random Volume Change Min (Decibels)", Next=@n3.On Next)
   DecibelstoLinearGain_0 = UE.Decibels to Linear Gain:Float(Decibels=@n4.Value)
   WavePlayer_0 = UE.Wave Player:Mono(Wave Asset=@n2.Value, Play=@n4.On Next, Pitch Shift=UE.Reroute(Value=@n3.Value).Value) {also {version [1,1]}}
   Multiply_0 = UE.Multiply:Audio by Float(AdditionalOperands=@DecibelstoLinearGain_0.Linear Gain, PrimaryOperand=@WavePlayer_0.Out Mono)
   n5 = InitVariable.Trigger()
   TriggerAnyTriggerAny2_0 = "TriggerAny.Trigger Any (2)"(
    In 1=VariableAccessor.Trigger(Variable=VariableMutator.Trigger(Variable=@n5.Variable, Value=@TriggerFilter_0.Tails).Variable).Value
    In 0=@WavePlayer_0.On Finished
   )
  ]
  links [{dst="@UE.OutputFormat.Mono.Audio:0", src=Multiply_0.Out}, {dst="@UE.Source.OneShot.OnFinished", src=TriggerAnyTriggerAny2_0.Out}]
 }]
properties {
 assetImportData=&self:AssetImportData
 attenuationSettings=/BuzzCool/Assets/Audio/Clones/Settings/ATT_BuzzCool_Voice
 bOverrideConcurrency=true
 maxDistance=2800
 sampleRate=48000
 soundClassObject=/Game/Sounds/FortSoundClassesAndModes/Fort_Events_Explosions_NPC
 volume=1.85
 concurrencyOverrides {bLimitToOwner=true, maxCount=3, resolutionRule=PreventNew}
 preEffectBusSends [{
   audioBus=/Game/Sounds/Modulation/EnvelopeFollowers/BR_Boss/AB_BR_Boss_VO_EnvFollow
   maxSendDistance=8000
   minSendDistance=0
   minSendLevel=1
   sendLevel=1
  }]
}
