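# Sound cue for a BuzzCool clone exploding, layering a character voice reaction with an explosion effect.
# A mixer plays a random "Doh" or NV_Various voice clip immediately, alongside a random explosion sample that starts after a short fixed delay.
# It routes to the NPC explosions sound class, uses the BuzzCool voice attenuation, and overrides concurrency to cap simultaneous instances.
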SoundCue /BuzzCool/Assets/Audio/Clones/SC_BuzzCool_Clone_Explode
root=Mixer_0
nodes [
 {id=Delay_0, children [Random_3], properties {delayMax=0.4, delayMin=0.4}}
 {id=Mixer_0, children [Delay_0,Random_0], properties {inputVolume [0.8,0.8]}}
 {
  id=Random_0
  children [WavePlayer_3, WavePlayer_4, WavePlayer_5, WavePlayer_6, WavePlayer_14, WavePlayer_15, WavePlayer_16, WavePlayer_19]
  properties {weights [1,1,1,1,1,1,1,1]}
 }
 {
  id=Random_3
  children [WavePlayer_0, WavePlayer_1, WavePlayer_2, WavePlayer_10, WavePlayer_11, WavePlayer_12]
  properties {weights [1,1,1,1,1,1]}
 }
 {id=WavePlayer_0, properties {soundWaveAssetPtr=/BuzzCool/Assets/Audio/wavs/BuzzCool_Explo_01}}
 {id=WavePlayer_1, properties {soundWaveAssetPtr=/BuzzCool/Assets/Audio/wavs/BuzzCool_Explo_02}}
 {id=WavePlayer_10, properties {soundWaveAssetPtr=/BuzzCool/Assets/Audio/wavs/BuzzCool_Explo_04}}
 {id=WavePlayer_11, properties {soundWaveAssetPtr=/BuzzCool/Assets/Audio/wavs/BuzzCool_Explo_05}}
 {id=WavePlayer_12, properties {soundWaveAssetPtr=/BuzzCool/Assets/Audio/wavs/BuzzCool_Explo_06}}
 {id=WavePlayer_14, properties {soundWaveAssetPtr=/BuzzCool/Assets/Audio/Clones/Waves/vo_IP_WildEstate_BuzzCool_Doh_010}}
 {id=WavePlayer_15, properties {soundWaveAssetPtr=/BuzzCool/Assets/Audio/Clones/Waves/vo_IP_WildEstate_BuzzCool_Doh_011}}
 {id=WavePlayer_16, properties {soundWaveAssetPtr=/BuzzCool/Assets/Audio/Clones/Waves/vo_IP_WildEstate_BuzzCool_Doh_012}}
 {id=WavePlayer_19, properties {soundWaveAssetPtr=/BuzzCool/Assets/Audio/Clones/Waves/vo_IP_WildEstate_BuzzCool_Doh_017}}
 {id=WavePlayer_2, properties {soundWaveAssetPtr=/BuzzCool/Assets/Audio/wavs/BuzzCool_Explo_03}}
 {id=WavePlayer_3, properties {soundWaveAssetPtr=/BuzzCool/Assets/Audio/Clones/Waves/vo_IP_WildEstate_BuzzCool_NV_Various_013}}
 {id=WavePlayer_4, properties {soundWaveAssetPtr=/BuzzCool/Assets/Audio/Clones/Waves/vo_IP_WildEstate_BuzzCool_NV_Various_012}}
 {id=WavePlayer_5, properties {soundWaveAssetPtr=/BuzzCool/Assets/Audio/Clones/Waves/vo_IP_WildEstate_BuzzCool_NV_Various_011}}
 {id=WavePlayer_6, properties {soundWaveAssetPtr=/BuzzCool/Assets/Audio/Clones/Waves/vo_IP_WildEstate_BuzzCool_NV_Various_010}}
]
properties {
 attenuationSettings=/BuzzCool/Assets/Audio/Clones/Settings/ATT_BuzzCool_Voice
 bBypassVolumeScaleForPriority=true
 bOverrideConcurrency=true
 maxDistance=2800
 priority=3
 soundClassObject=/Game/Sounds/FortSoundClassesAndModes/Fort_Events_Explosions_NPC
 volumeMultiplier=2
 attenuationOverrides {bEnableReverbSend=false, reverbWetLevelMax=0.3, reverbWetLevelMin=0.1}
 concurrencyOverrides {maxCount=6, volumeScale=0.9}
}
