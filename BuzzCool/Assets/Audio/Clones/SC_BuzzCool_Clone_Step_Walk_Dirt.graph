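# Sound Cue for clone walk-on-dirt footsteps in BuzzCool, though its samples are twelve Soldier run-on-stone step waves.
# An equally weighted Random node picks one wave and feeds a root Modulator that varies volume and slightly raises pitch.
# Playback uses Deimos step attenuation and the player/enemy movement sound class, with virtualization disabled and a distance-scaled concurrency override capping simultaneous steps.
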SoundCue /BuzzCool/Assets/Audio/Clones/SC_BuzzCool_Clone_Step_Walk_Dirt
root=Modulator_0
nodes [
 {id=Modulator_0, children [Random_0], properties {pitchMax=1.07, pitchMin=1, volumeMax=1, volumeMin=0.6}}
 {
  id=Random_0
  children [
   WavePlayer_0
   WavePlayer_1
   WavePlayer_2
   WavePlayer_3
   WavePlayer_4
   WavePlayer_5
   WavePlayer_6
   WavePlayer_24
   WavePlayer_25
   WavePlayer_26
   WavePlayer_27
   WavePlayer_28
  ]
  properties {weights [1,1,1,1,1,1,1,1,1,1,1,1]}
 }
 {id=WavePlayer_0, properties {soundWaveAssetPtr=/Game/Sounds/Fort_Step_Sounds/Soldier/Soldier_Step_Run_Stone_10}}
 {id=WavePlayer_1, properties {soundWaveAssetPtr=/Game/Sounds/Fort_Step_Sounds/Soldier/Soldier_Step_Run_Stone_11}}
 {id=WavePlayer_2, properties {soundWaveAssetPtr=/Game/Sounds/Fort_Step_Sounds/Soldier/Soldier_Step_Run_Stone_12}}
 {id=WavePlayer_24, properties {soundWaveAssetPtr=/Game/Sounds/Fort_Step_Sounds/Soldier/Soldier_Step_Run_Stone_05}}
 {id=WavePlayer_25, properties {soundWaveAssetPtr=/Game/Sounds/Fort_Step_Sounds/Soldier/Soldier_Step_Run_Stone_06}}
 {id=WavePlayer_26, properties {soundWaveAssetPtr=/Game/Sounds/Fort_Step_Sounds/Soldier/Soldier_Step_Run_Stone_07}}
 {id=WavePlayer_27, properties {soundWaveAssetPtr=/Game/Sounds/Fort_Step_Sounds/Soldier/Soldier_Step_Run_Stone_08}}
 {id=WavePlayer_28, properties {soundWaveAssetPtr=/Game/Sounds/Fort_Step_Sounds/Soldier/Soldier_Step_Run_Stone_09}}
 {id=WavePlayer_3, properties {soundWaveAssetPtr=/Game/Sounds/Fort_Step_Sounds/Soldier/Soldier_Step_Run_Stone_01}}
 {id=WavePlayer_4, properties {soundWaveAssetPtr=/Game/Sounds/Fort_Step_Sounds/Soldier/Soldier_Step_Run_Stone_02}}
 {id=WavePlayer_5, properties {soundWaveAssetPtr=/Game/Sounds/Fort_Step_Sounds/Soldier/Soldier_Step_Run_Stone_03}}
 {id=WavePlayer_6, properties {soundWaveAssetPtr=/Game/Sounds/Fort_Step_Sounds/Soldier/Soldier_Step_Run_Stone_04}}
]
properties {
 attenuationSettings=/Game/Sounds/Attenuation/Fort_Deimos_Step_Att
 bBypassVolumeScaleForPriority=true
 bOverrideConcurrency=true
 maxDistance=2000
 priority=0
 soundClassObject=/Game/Sounds/FortSoundClassesAndModes/Fort_09_Player_Enemy_Movement
 virtualizationMode=Disabled
 volumeMultiplier=0.85
 concurrencyOverrides {maxCount=10, volumeScaleMode=Distance}
}
