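# Sound cue for a BuzzCool clone's melee swing hit, layering a punch impact with a punch whoosh.
# Two Random nodes, one picking among six punches and one among nine A/B whooshes, feed a boosted Mixer so both play together.
# A root Modulator adds per-play variation to the mix.
# The cue uses the Fort_Creatures sound class, Deimos melee attack attenuation and player melee weapon concurrency, and primes on load.
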SoundCue /BuzzCool/Assets/Audio/Clones/SC_BuzzCool_Clone_SwingHIt
root=Modulator_1
nodes [
 {id=Mixer_0, children [Random_1,Random_3], properties {inputVolume [1.5,1.5]}}
 {id=Modulator_1, children [Mixer_0], properties {pitchMin=1}}
 {id=Random_1, children [WavePlayer_2,WavePlayer_5,WavePlayer_6,WavePlayer_7,WavePlayer_8,WavePlayer_9], properties {weights [1,1,1,1,1,1]}}
 {
  id=Random_3
  children [WavePlayer_1, WavePlayer_0, WavePlayer_3, WavePlayer_4, WavePlayer_10, WavePlayer_11, WavePlayer_12, WavePlayer_13, WavePlayer_14]
  properties {weights [1,1,1,1,1,1,1,1,1]}
 }
 {id=WavePlayer_0, properties {soundWaveAssetPtr=/BuzzCool/Assets/Audio/wavs/BuzzCool_PunchWhooshB_02}}
 {id=WavePlayer_1, properties {soundWaveAssetPtr=/BuzzCool/Assets/Audio/wavs/BuzzCool_PunchWhooshB_01}}
 {id=WavePlayer_10, properties {soundWaveAssetPtr=/BuzzCool/Assets/Audio/wavs/BuzzCool_PunchWhooshA_01}}
 {id=WavePlayer_11, properties {soundWaveAssetPtr=/BuzzCool/Assets/Audio/wavs/BuzzCool_PunchWhooshA_02}}
 {id=WavePlayer_12, properties {soundWaveAssetPtr=/BuzzCool/Assets/Audio/wavs/BuzzCool_PunchWhooshA_03}}
 {id=WavePlayer_13, properties {soundWaveAssetPtr=/BuzzCool/Assets/Audio/wavs/BuzzCool_PunchWhooshA_04}}
 {id=WavePlayer_14, properties {soundWaveAssetPtr=/BuzzCool/Assets/Audio/wavs/BuzzCool_PunchWhooshA_05}}
 {id=WavePlayer_2, properties {soundWaveAssetPtr=/BuzzCool/Assets/Audio/wavs/BuzzCool_Punch_01}}
 {id=WavePlayer_3, properties {soundWaveAssetPtr=/BuzzCool/Assets/Audio/wavs/BuzzCool_PunchWhooshB_03}}
 {id=WavePlayer_4, properties {soundWaveAssetPtr=/BuzzCool/Assets/Audio/wavs/BuzzCool_PunchWhooshB_04}}
 {id=WavePlayer_5, properties {soundWaveAssetPtr=/BuzzCool/Assets/Audio/wavs/BuzzCool_Punch_02}}
 {id=WavePlayer_6, properties {soundWaveAssetPtr=/BuzzCool/Assets/Audio/wavs/BuzzCool_Punch_03}}
 {id=WavePlayer_7, properties {soundWaveAssetPtr=/BuzzCool/Assets/Audio/wavs/BuzzCool_Punch_04}}
 {id=WavePlayer_8, properties {soundWaveAssetPtr=/BuzzCool/Assets/Audio/wavs/BuzzCool_Punch_05}}
 {id=WavePlayer_9, properties {soundWaveAssetPtr=/BuzzCool/Assets/Audio/wavs/BuzzCool_Punch_06}}
]
properties {
 attenuationSettings=/Game/Sounds/Attenuation/Fort_Deimos_Attack_Melee_Att
 bPrimeOnLoad=true
 maxDistance=3300
 soundClassObject=/Game/Sounds/FortSoundClassesAndModes/Fort_Creatures
 volumeMultiplier=1
 attenuationOverrides {
  bAttenuateWithLPF=true
  distanceAlgorithm=Logarithmic
  falloffDistance=800
  lPFRadiusMax=800
  lPFRadiusMin=400
  attenuationShapeExtents {x=300}
 }
 concurrencyOverrides {bLimitToOwner=true, maxCount=2, resolutionRule=StopOldest}
 concurrencySet [&Player_Melee_Weapon_Conc]
}
