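# Sound cue for the BuzzCool clone's swing kick, playing one of four punch whoosh waves with slight pitch and volume variation.
# IsLocal and IsEnemy branches route local or unset cases to the base level, remote enemies to a boosted mixer and remote non-enemies to a reduced one.
# All paths share the same waves; the cue uses melee attack attenuation and overrides concurrency to cap simultaneous instances.
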SoundCue /BuzzCool/Assets/Audio/Clones/SC_BuzzCool_Clone_SwingKick
root=Branch_0
nodes [
 {id=Branch_0, children [Modulator_0,Branch_1,Modulator_0], properties {boolParameterName=IsLocal}}
 {id=Branch_1, children [Mixer_8,Mixer_6,Modulator_0], properties {boolParameterName=IsEnemy}}
 {id=Mixer_6, children [Modulator_3], properties {inputVolume [0.65]}}
 {id=Mixer_8, children [Modulator_2], properties {inputVolume [1.3]}}
 {id=Modulator_0, children [Random_0], properties {pitchMax=1.1, pitchMin=1.05, volumeMin=0.85}}
 {id=Modulator_2, children [Random_2], properties {pitchMax=1.1, pitchMin=1.05, volumeMin=0.85}}
 {id=Modulator_3, children [Random_3], properties {pitchMax=1.1, pitchMin=1.05, volumeMin=0.85}}
 {id=Random_0, children [WavePlayer_1,WavePlayer_2,WavePlayer_3,WavePlayer_4], properties {weights [1,1,1,1]}}
 {id=Random_2, children [WavePlayer_5, WavePlayer_15, WavePlayer_16, WavePlayer_17], properties {weights [1,1,1,1]}}
 {id=Random_3, children [WavePlayer_18, WavePlayer_19, WavePlayer_20, WavePlayer_21], properties {weights [1,1,1,1]}}
 {id=WavePlayer_1, properties {soundWaveAssetPtr=/BuzzCool/Assets/Audio/wavs/BuzzCool_PunchWhooshB_01}}
 {id=WavePlayer_15, properties {soundWaveAssetPtr=/BuzzCool/Assets/Audio/wavs/BuzzCool_PunchWhooshB_02}}
 {id=WavePlayer_16, properties {soundWaveAssetPtr=/BuzzCool/Assets/Audio/wavs/BuzzCool_PunchWhooshB_03}}
 {id=WavePlayer_17, properties {soundWaveAssetPtr=/BuzzCool/Assets/Audio/wavs/BuzzCool_PunchWhooshB_04}}
 {id=WavePlayer_18, properties {soundWaveAssetPtr=&BuzzCool_PunchWhooshB_01}}
 {id=WavePlayer_19, properties {soundWaveAssetPtr=&BuzzCool_PunchWhooshB_02}}
 {id=WavePlayer_2, properties {soundWaveAssetPtr=&BuzzCool_PunchWhooshB_02}}
 {id=WavePlayer_20, properties {soundWaveAssetPtr=&BuzzCool_PunchWhooshB_03}}
 {id=WavePlayer_21, properties {soundWaveAssetPtr=&BuzzCool_PunchWhooshB_04}}
 {id=WavePlayer_3, properties {soundWaveAssetPtr=&BuzzCool_PunchWhooshB_03}}
 {id=WavePlayer_4, properties {soundWaveAssetPtr=&BuzzCool_PunchWhooshB_04}}
 {id=WavePlayer_5, properties {soundWaveAssetPtr=&BuzzCool_PunchWhooshB_01}}
]
properties {
 attenuationSettings=/Game/Sounds/Attenuation/Fort_Deimos_Attack_Melee_Att
 bOverrideConcurrency=true
 maxDistance=3300
 soundClassObject=&Fort_WeaponRanged_Fire_3P_PSM
 volumeMultiplier=1.3
 busSends [{audioBus=null, maxSendDistance=0, minSendDistance=1000, minSendLevel=0, sendLevel=1, soundSourceBus=null}]
 concurrencyOverrides {maxCount=3}
}
