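# Sound cue for the BuzzCool clone's punch swing whoosh, with melee attack attenuation and overridden concurrency.
# An IsLocal branch plays local or unset cases at base level and sends non-local ones to an IsEnemy branch.
# That branch boosts enemy swings and lowers non-enemy ones through mixers, falling back to base level when unset.
# Every path applies volume modulation to an equally weighted random pick from the same five whoosh waves.
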SoundCue /BuzzCool/Assets/Audio/Clones/SC_BuzzCool_Clone_SwingPunch
root=Branch_0
nodes [
 {id=Branch_0, children [Modulator_0,Branch_1,Modulator_0], properties {boolParameterName=IsLocal}}
 {id=Branch_1, children [Mixer_8,Mixer_6,Modulator_0], properties {boolParameterName=IsEnemy}}
 {id=Mixer_6, children [Modulator_3], properties {inputVolume [0.65]}}
 {id=Mixer_8, children [Modulator_2], properties {inputVolume [1.3]}}
 {id=Modulator_0, children [Random_1], properties {volumeMin=0.85}}
 {id=Modulator_2, children [Random_4], properties {volumeMin=0.85}}
 {id=Modulator_3, children [Random_5], properties {volumeMin=0.85}}
 {id=Random_1, children [WavePlayer_0,WavePlayer_6,WavePlayer_7,WavePlayer_8,WavePlayer_9], properties {weights [1,1,1,1,1]}}
 {id=Random_4, children [WavePlayer_10, WavePlayer_11, WavePlayer_12, WavePlayer_13, WavePlayer_14], properties {weights [1,1,1,1,1]}}
 {id=Random_5, children [WavePlayer_22, WavePlayer_23, WavePlayer_27, WavePlayer_28, WavePlayer_29], properties {weights [1,1,1,1,1]}}
 {id=WavePlayer_0, properties {soundWaveAssetPtr=/BuzzCool/Assets/Audio/wavs/BuzzCool_PunchWhooshA_01}}
 {id=WavePlayer_10, properties {soundWaveAssetPtr=&BuzzCool_PunchWhooshA_01}}
 {id=WavePlayer_11, properties {soundWaveAssetPtr=/BuzzCool/Assets/Audio/wavs/BuzzCool_PunchWhooshA_02}}
 {id=WavePlayer_12, properties {soundWaveAssetPtr=/BuzzCool/Assets/Audio/wavs/BuzzCool_PunchWhooshA_03}}
 {id=WavePlayer_13, properties {soundWaveAssetPtr=/BuzzCool/Assets/Audio/wavs/BuzzCool_PunchWhooshA_04}}
 {id=WavePlayer_14, properties {soundWaveAssetPtr=/BuzzCool/Assets/Audio/wavs/BuzzCool_PunchWhooshA_05}}
 {id=WavePlayer_22, properties {soundWaveAssetPtr=&BuzzCool_PunchWhooshA_01}}
 {id=WavePlayer_23, properties {soundWaveAssetPtr=&BuzzCool_PunchWhooshA_02}}
 {id=WavePlayer_27, properties {soundWaveAssetPtr=&BuzzCool_PunchWhooshA_03}}
 {id=WavePlayer_28, properties {soundWaveAssetPtr=&BuzzCool_PunchWhooshA_04}}
 {id=WavePlayer_29, properties {soundWaveAssetPtr=&BuzzCool_PunchWhooshA_05}}
 {id=WavePlayer_6, properties {soundWaveAssetPtr=&BuzzCool_PunchWhooshA_02}}
 {id=WavePlayer_7, properties {soundWaveAssetPtr=&BuzzCool_PunchWhooshA_03}}
 {id=WavePlayer_8, properties {soundWaveAssetPtr=&BuzzCool_PunchWhooshA_04}}
 {id=WavePlayer_9, properties {soundWaveAssetPtr=&BuzzCool_PunchWhooshA_05}}
]
properties {
 attenuationSettings=/Game/Sounds/Attenuation/Fort_Deimos_Attack_Melee_Att
 bOverrideConcurrency=true
 maxDistance=3300
 soundClassObject=&Fort_WeaponRanged_Fire_3P_PSM
 volumeMultiplier=1.2
 busSends [{audioBus=null, maxSendDistance=0, minSendDistance=1000, minSendLevel=0, sendLevel=1, soundSourceBus=null}]
 concurrencyOverrides {maxCount=3}
}
