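# Sound cue for BuzzCool clones falling, layering a random voice grunt with a random body impact.
# One random node picks from burp, jump and throw vocal clips and another picks from six body impact waves, all equally weighted.
# A mixer plays both picks together through a modulator at the root.
# It uses voice attenuation and the events VO sound class, with a concurrency override that blocks new instances past a small cap.
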SoundCue /BuzzCool/Assets/Audio/Clones/SC_BuzzCool_Clones_Falling
root=Modulator_0
nodes [
 {id=Mixer_0, children [Random_3,Random_1], properties {inputVolume [1,1]}}
 {id=Modulator_0, children [Mixer_0]}
 {
  id=Random_1
  children [WavePlayer_3, WavePlayer_4, WavePlayer_9, WavePlayer_11, WavePlayer_12, WavePlayer_13]
  properties {weights [1,1,1,1,1,1]}
 }
 {
  id=Random_3
  children [WavePlayer_8, WavePlayer_10, WavePlayer_18, WavePlayer_19, WavePlayer_21, WavePlayer_22, WavePlayer_23]
  properties {weights [1,1,1,1,1,1,1]}
 }
 {id=WavePlayer_10, properties {soundWaveAssetPtr=/BuzzCool/Assets/Audio/Clones/Waves/vo_IP_WildEstate_GiantHomer_NV_Burp_011}}
 {id=WavePlayer_11, properties {soundWaveAssetPtr=/BuzzCool/Assets/Audio/wavs/BuzzCool_Body_Impac_04}}
 {id=WavePlayer_12, properties {soundWaveAssetPtr=/BuzzCool/Assets/Audio/wavs/BuzzCool_Body_Impac_05}}
 {id=WavePlayer_13, properties {soundWaveAssetPtr=/BuzzCool/Assets/Audio/wavs/BuzzCool_Body_Impac_06}}
 {id=WavePlayer_18, properties {soundWaveAssetPtr=/BuzzCool/Assets/Audio/Clones/Waves/vo_IP_WildEstate_SilentNovel_NV_Jump_014}}
 {id=WavePlayer_19, properties {soundWaveAssetPtr=/BuzzCool/Assets/Audio/Clones/Waves/vo_IP_WildEstate_SilentNovel_NV_Throw_010}}
 {id=WavePlayer_21, properties {soundWaveAssetPtr=/BuzzCool/Assets/Audio/Clones/Waves/vo_IP_WildEstate_SilentNovel_NV_Throw_012}}
 {id=WavePlayer_22, properties {soundWaveAssetPtr=/BuzzCool/Assets/Audio/Clones/Waves/vo_IP_WildEstate_SilentNovel_NV_Throw_013}}
 {id=WavePlayer_23, properties {soundWaveAssetPtr=/BuzzCool/Assets/Audio/Clones/Waves/vo_IP_WildEstate_SilentNovel_NV_Throw_014}}
 {id=WavePlayer_3, properties {soundWaveAssetPtr=/BuzzCool/Assets/Audio/wavs/BuzzCool_Body_Impac_01}}
 {id=WavePlayer_4, properties {soundWaveAssetPtr=/BuzzCool/Assets/Audio/wavs/BuzzCool_Body_Impac_02}}
 {id=WavePlayer_8, properties {soundWaveAssetPtr=/BuzzCool/Assets/Audio/Clones/Waves/vo_IP_WildEstate_GiantHomer_NV_Burp_010}}
 {id=WavePlayer_9, properties {soundWaveAssetPtr=/BuzzCool/Assets/Audio/wavs/BuzzCool_Body_Impac_03}}
]
properties {
 attenuationSettings=/BuzzCool/Assets/Audio/Clones/Settings/ATT_BuzzCool_Voice
 maxDistance=2800
 soundClassObject=/Game/Sounds/FortSoundClassesAndModes/Fort_Events_VO
 volumeMultiplier=0.2807
 concurrencyOverrides {maxCount=4, resolutionRule=PreventNew}
}
