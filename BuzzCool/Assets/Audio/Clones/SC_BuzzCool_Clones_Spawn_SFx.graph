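# Sound cue for the BuzzCool clone spawn effect, layering a rift spawn sound with a delayed character vocalization.
# A Mixer blends a quieter random Lager rift wave with a full-volume random burp, jump or throw voice clip, which follows after a short randomized delay.
# A root Modulator at default settings processes the mix.
# It borrows explosion attenuation and sound class, with concurrency overrides set to a small voice cap that rejects new plays.
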SoundCue /BuzzCool/Assets/Audio/Clones/SC_BuzzCool_Clones_Spawn_SFx
root=Modulator_0
nodes [
 {id=Delay_0, children [Random_3], properties {delayMax=0.45, delayMin=0.35}}
 {id=Mixer_0, children [Delay_0,Random_0], properties {inputVolume [1,0.4]}}
 {id=Modulator_0, children [Mixer_0]}
 {id=Random_0, children [WavePlayer_0,WavePlayer_1,WavePlayer_2,WavePlayer_5,WavePlayer_6,WavePlayer_7], properties {weights [1,1,1,1,1,1]}}
 {
  id=Random_3
  children [
   WavePlayer_8
   WavePlayer_10
   WavePlayer_15
   WavePlayer_17
   WavePlayer_18
   WavePlayer_19
   WavePlayer_20
   WavePlayer_21
   WavePlayer_22
   WavePlayer_23
  ]
  properties {weights [1,1,1,1,1,1,1,1,1,1]}
 }
 {id=WavePlayer_0, properties {soundWaveAssetPtr=/BuzzCool/Assets/Audio/wavs/Lager_Spawn_Rift_03}}
 {id=WavePlayer_1, properties {soundWaveAssetPtr=/BuzzCool/Assets/Audio/wavs/Lager_Spawn_Rift_04}}
 {id=WavePlayer_10, properties {soundWaveAssetPtr=/BuzzCool/Assets/Audio/Clones/Waves/vo_IP_WildEstate_GiantHomer_NV_Burp_011}}
 {id=WavePlayer_15, properties {soundWaveAssetPtr=/BuzzCool/Assets/Audio/Clones/Waves/vo_IP_WildEstate_SilentNovel_NV_Jump_011}}
 {id=WavePlayer_17, properties {soundWaveAssetPtr=/BuzzCool/Assets/Audio/Clones/Waves/vo_IP_WildEstate_SilentNovel_NV_Jump_013}}
 {id=WavePlayer_18, properties {soundWaveAssetPtr=/BuzzCool/Assets/Audio/Clones/Waves/vo_IP_WildEstate_SilentNovel_NV_Jump_014}}
 {id=WavePlayer_19, properties {soundWaveAssetPtr=/BuzzCool/Assets/Audio/Clones/Waves/vo_IP_WildEstate_SilentNovel_NV_Throw_010}}
 {id=WavePlayer_2, properties {soundWaveAssetPtr=/BuzzCool/Assets/Audio/wavs/Lager_Spawn_Rift_05}}
 {id=WavePlayer_20, properties {soundWaveAssetPtr=/BuzzCool/Assets/Audio/Clones/Waves/vo_IP_WildEstate_SilentNovel_NV_Throw_011}}
 {id=WavePlayer_21, properties {soundWaveAssetPtr=/BuzzCool/Assets/Audio/Clones/Waves/vo_IP_WildEstate_SilentNovel_NV_Throw_012}}
 {id=WavePlayer_22, properties {soundWaveAssetPtr=/BuzzCool/Assets/Audio/Clones/Waves/vo_IP_WildEstate_SilentNovel_NV_Throw_013}}
 {id=WavePlayer_23, properties {soundWaveAssetPtr=/BuzzCool/Assets/Audio/Clones/Waves/vo_IP_WildEstate_SilentNovel_NV_Throw_014}}
 {id=WavePlayer_5, properties {soundWaveAssetPtr=/BuzzCool/Assets/Audio/wavs/Lager_Spawn_Rift_06}}
 {id=WavePlayer_6, properties {soundWaveAssetPtr=/BuzzCool/Assets/Audio/wavs/Lager_Spawn_Rift_07}}
 {id=WavePlayer_7, properties {soundWaveAssetPtr=/BuzzCool/Assets/Audio/wavs/Lager_Spawn_Rift_08}}
 {id=WavePlayer_8, properties {soundWaveAssetPtr=/BuzzCool/Assets/Audio/Clones/Waves/vo_IP_WildEstate_GiantHomer_NV_Burp_010}}
]
properties {
 attenuationSettings=/BuzzCool/Assets/Audio/Clones/Settings/ATT_BuzzCool_Explode
 maxDistance=16512
 soundClassObject=/Game/Sounds/FortSoundClassesAndModes/Fort_Events_Explosions_NPC
 volumeMultiplier=0.8
 concurrencyOverrides {maxCount=4, resolutionRule=PreventNew}
}
